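# Translucent unlit Niagara ribbon material for the Curie drying smoke effect.
# A smoke trail texture pans at a dynamic-parameter U speed, is desaturated, then tinted by particle color through an emissive adjustment function.
# The same texture, scaled by particle alpha, camera depth fade and a view-angle dot-product fade, has dynamic Erode subtracted for dissolving.
# The view fade compares camera direction from actor or object position against a local-space vector.
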Material /BRCosmetics/Curie/Effects/M_Curie_DryingSmokeRibbon
graph {
 properties {bUsedWithNiagaraRibbons=true, blendMode=BLEND_Translucent, pixelDepthOffsetMode=PDOM_Legacy, shadingModel=MSM_Unlit, usedShadingModels=Unlit}
 nodes [
  n1 = AppendVector(A=DynamicParameter() {paramNames [PanSpeedU,Erode,Param3,Param4]}.PanSpeedU, B=Constant())
  n2 = Desaturation(
   None=TextureSampleParameter2D(UVs=Panner(Speed=@n1) {bFractionalPart=true, speedX=-0.5}) {parameterName=Texture, texture=/Game/Effects/Fort_Effects/Textures/Smoke/T_Smoke_Trail_Soft_2}.RGB
   Fraction=ScalarParameter() {parameterName=Desaturation Amt}
  )
  n3 = ParticleColor()
  MaterialFunctionCall(Emissive=Multiply(A=@n2, B=@n3.RGB)) {materialFunction=&MF_FX_EmissiveAdjustments}
  n4 = Constant() {r=2}
  n5 = Constant()
  Saturate(None=LinearInterpolate(A=Constant() {r=1}, B=Constant() {r=-1}, Alpha=Fresnel(ExponentIn=@n4, BaseReflectFractionIn=@n5) {baseReflectFraction=0}))
  n6 = MaterialFunctionCall(Fade Length=ScalarParameter() {parameterName=Cam Fade length}, Fade Offset=ScalarParameter() {parameterName=Cam Fade offset}) {materialFunction=&CameraDepthFade}
  n7 = ActorPositionWS()
  n8 = ObjectPositionWS()
  n9 = ScalarParameter() {defaultValue=1, parameterName=Actor vs obcet position}
  n10 = VectorParameter() {parameterName=Actor/position Offset, channelNames {a=Is Local, b=Z, g=Y, r=X}, defaultValue=<rgba 0 0 0 1>}
  n11 = Transform(None=@n10) {transformSourceType=TRANSFORMSOURCE_Local}
  n12 = Subtract(A=CameraPositionWS(), B=Add(A=LinearInterpolate(A=@n7, B=@n8, Alpha=@n9), B=LinearInterpolate(A=@n10, B=@n11, Alpha=@n10)))
  n13 = VectorParameter() {parameterName=Object vector to comapre, defaultValue=<rgba 0 1 0 1>}
  n14 = ConstantBiasScale(None=DotProduct(A=Normalize(VectorInput=@n12), B=Transform(None=@n13.RGB) {transformSourceType=TRANSFORMSOURCE_Local}))
  n15 = VectorParameter() {parameterName=Dot Options, channelNames {a=pre mul, b=add, g=mul, r=pow}, defaultValue=<rgba 1 1 0 1>}
  n16 = Clamp(None=Multiply(A=OneMinus(None=@n14), B=@n15.pre mul))
  n17 = Add(A=Multiply(A=Power(Base=@n16, Exp=@n15.pow) {constExponent=1}, B=@n15.mul), B=@n15.add) {constB=0}
  n18 = Clamp(None=LinearInterpolate(
    A=ScalarParameter() {defaultValue=1, parameterName=opacity - look at}
    B=ScalarParameter() {defaultValue=1, parameterName=opacity - look away}
    Alpha=Clamp(None=@n17)
   ))
  Subtract(
   A=Multiply(A=Multiply(A=@n2, B=@n3.A), B=Multiply(A=@n6.Result, B=@n18))
   B=DynamicParameter() {paramNames [PanSpeedU,Erode,Param3,Param4]}.Erode
  )
 ]
}
